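# Editor utility widget for a modular asset browser in the AnimInEdTools editor tools.
# On construct, it finds BP_AssetBrowserModule subclasses through the Asset Registry, keeping non-abstract ones as modules for a combo box.
# Picking a module lists its asset names; picking an asset hosts that module's widget in a named slot.
# Add opens a modal around the module's BPI_AssetBrowserCreate widget, which validates input, then creates and saves a new asset.
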EditorUtilityWidgetBlueprint /AnimInEdTools/AssetBrowser/AssetBrowser
graphs [
 Create Modal Dialog {nodes [
   Create Modal Dialog!() -> (
    then
    Widget:Editor Utility Widget Object Reference
    Create Button Delegate:"Delegate (by ref)"
    Cancel Button Delegate:"Delegate (by ref)"
    Title:String
   ) {is=title K2Node_FunctionEntry}
   > n1 = Create Widget!(
    execute:Exec
    WorldContextObject:Object Reference
    Class:User Widget Class Reference=/AnimInEdTools/AssetBrowser/Blueprint/EUW_AssetBrowser_Modal.EUW_AssetBrowser_Modal_C
    OwningPlayer:Player Controller Object Reference
   ) -> (then:Exec, ReturnValue:EUW Asset Browser Modal Object Reference)
   > n2 = Sequence()
   > Set Text(
    Text="To Text (String)"(InString=Get Title() -> (Title:String))
    self=Get Title_TextBlock!(self:EUW Asset Browser Modal Object Reference=@n1) -> (Title_TextBlock:Text Object Reference)
   )
   > AddChild(
    self=Get NamedSlot!(self:EUW Asset Browser Modal Object Reference=@n1) -> (NamedSlot:Named Slot Object Reference)
    Content=Get Widget() -> (Widget:Editor Utility Widget Object Reference)
   )
   > n3 = AddChildToCanvas!(
    execute
    self:Canvas Panel Object Reference=Get CanvasPanel() -> (CanvasPanel:Canvas Panel Object Reference)
    Content:Widget Object Reference=@n1
   ) -> (then, ReturnValue:Canvas Panel Slot Object Reference)
   > SetAnchors!(
    execute
    self:Canvas Panel Slot Object Reference=@n3
    InAnchors:Anchors Structure=Make Anchors!(Minimum:Vector 2D Structure=(X=0,Y=0), Maximum:Vector 2D Structure=(X=1,Y=1)) -> (Anchors:Anchors Structure) {is=title K2Node_MakeStruct}
   ) -> (then)
   > SetZOrder!(execute, self:Canvas Panel Slot Object Reference=@n3, InZOrder:int=1) -> (then)
   > SetOffsets!(execute, self:Canvas Panel Slot Object Reference=@n3, InOffset:Margin Structure=Make Margin(Left=0, Top=0, Right=0, Bottom=0)) -> (then)
   Bind Event to On Clicked(
    execute=@n2.then_1
    self:Button Object Reference=Get CancelButton!(self:EUW Asset Browser Modal Object Reference=@n1) -> (CancelButton:Button Object Reference)
    Delegate=Get Cancel Button Delegate() -> (Cancel Button Delegate:"Delegate (by ref)")
   ) {is=title K2Node_AddDelegate}
   > Bind Event to On Clicked(
    self=Get CreateButton!(self:EUW Asset Browser Modal Object Reference=@n1) -> (CreateButton:Button Object Reference)
    Delegate=Get Create Button Delegate() -> (Create Button Delegate:"Delegate (by ref)")
   )
   > Return Node!(execute, Modal Widget:EUW Asset Browser Modal Object Reference=@n1)
  ]}
 EventGraph {nodes [
   n1 = "On Selection Changed (Module_ComboBox)"!() -> (OutputDelegate, then, SelectedItem:String, SelectionType:ESelectInfo Enum) {is=title K2Node_ComponentBoundEvent}
   > ClearChildren(self=Get NamedSlot!() -> (NamedSlot))
   > ClearOptions(self:"Searchable Combo Box (String) Object Reference"=Get Asset_ComboBox() -> (Asset_ComboBox:"Searchable Combo Box (String) Object Reference"))
   > n2 = Set Current Module(Current Module:BP Asset Browser Module Object Reference=Find Module by Name!(
     Modules:Array of BP Asset Browser Module Object References=Get Modules() -> (Modules:Array of BP Asset Browser Module Object References)
     Name:String=@n1.SelectedItem
    ) -> (Module:BP Asset Browser Module Object Reference)) -> (Output_Get:BP Asset Browser Module Object Reference)
   n3 = Get Create Widget!(self:BP Asset Browser Module Object Reference=@n2) -> (Create Widget:Editor Utility Widget Soft Class Reference)
   > SetVisibility(
    self=Get Add_Button() -> (Add_Button:Editor Utility Button Object Reference)
    InVisibility=Select(
     Option 0=Collapsed
     Option 1=Visible
     Index=IsValidSoftClassReference!(SoftClassReference:"Object Soft Class Reference (by ref)"=@n3) -> (ReturnValue:bool)
    )
   )
   > n4 = Get Asset Names!(execute, self:BP Asset Browser Module Object Reference=Get Current Module() -> (Current Module:BP Asset Browser Module Object Reference)) -> (then, Assets:Array of Strings) {is=title K2Node_CallFunction}
   > n5 = For Each Loop(Array=@n4)
   > AddOption!(execute, self:"Searchable Combo Box (String) Object Reference"=Get Asset_ComboBox(), Option:String=@n5) -> (then)
   "On Clicked (Add_Button)"() {is=title K2Node_ComponentBoundEvent}
   > n6 = Get Create Widget!(execute, self:BP Asset Browser Module Object Reference=Get Current Module()) -> (then, Create Widget:Editor Utility Widget Object Reference) {is=title K2Node_CallFunction}
   > n7 = Branch(Condition=Does Object Implement Interface!(
     TestObject:Object Reference=@n6
     Interface:Interface Interface=/AnimInEdTools/AssetBrowser/Blueprint/BPI_AssetBrowserCreate.BPI_AssetBrowserCreate_C
    ) -> (ReturnValue:bool))
   > n8 = Get Title!(execute:Exec, self:Object Reference=@n6) -> (then:Exec, Title:Name=None) {is=title K2Node_Message}
   n9 = Modal Dialog Cancel() {is=title K2Node_CustomEvent}
   n10 = Modal Dialog Create() {is=title K2Node_CustomEvent}
   n11 = Create Modal Dialog!(
    execute=@n8.then
    self
    Widget:Editor Utility Widget Object Reference=@n6
    Create Button Delegate:"Delegate (by ref)"=@n10
    Cancel Button Delegate:"Delegate (by ref)"=@n9
    Title:String="To String (Name)"(InName=@n8)
   ) -> (then, Modal Widget:EUW Asset Browser Modal Object Reference) {is=title K2Node_CallFunction}
   PrintString(execute=@n7.else, InString=Create Widget does not implement BPI_AssetBrowserCreate)
   n12 = "On Selection Changed (Asset_ComboBox)"!() -> (OutputDelegate, then, SelectedItem:String, SelectionType:ESelectInfo Enum) {is=title K2Node_ComponentBoundEvent}
   n13 = Get Module Widget!(execute=[@n12.then, @n16.then], self:BP Asset Browser Module Object Reference=Get Current Module()) -> (then, Module Widget:Editor Utility Widget Object Reference) {is=title K2Node_CallFunction}
   > n14 = Branch(Condition=HasChild!(self:Panel Widget Object Reference=Get NamedSlot(), Content:Widget Object Reference=@n13) -> (ReturnValue:bool))
   Set Current Asset!(execute=[@n15.then, @n14.then], self:BP Asset Browser Module Object Reference=Get Current Module(), Asset:String=@n12.SelectedItem) -> (then) {is=title K2Node_CallFunction}
   n15 = AddChild(execute=@n14.else, self=Get NamedSlot(), Content=@n13)
   n16 = Updated Asset Widget() {is=title K2Node_CustomEvent}
   n17 = Create Dialog Cancel Button Clicked!() -> (OutputDelegate, then, "Modal  Widget":Widget Object Reference) {is=title K2Node_CustomEvent}
   > RemoveChild(self=Get CanvasPanel(), Content=@n17."Modal  Widget")
   n18 = Create Dialog Create Clicked!() -> (OutputDelegate, then, Create Widget:Widget Object Reference, Modal Dialog Widget:EUW Asset Browser Modal Object Reference) {is=title K2Node_CustomEvent}
   > n19 = Validate Data!(execute:Exec, self:Object Reference=@n18.Create Widget) -> (then:Exec, Data is Valid:bool, Text:Text) {is=title K2Node_Message}
   > n20 = Branch(Condition=@n19)
   > n21 = Sequence()
   > n22 = Create Breadcrumb!(execute:Exec, self:Object Reference=@n18.Create Widget) -> (then:Exec, Breadcrumb:ABT Abstract Object Reference) {is=title K2Node_Message}
   > Is Valid(InputObject=@n22)
   > SaveLoadedAsset(AssetToSave=@n22, bOnlyIfIsDirty=true)
   n23 = Get Name!(self:ABT Abstract Object Reference=@n22) -> (Name:String)
   n24 = Get Asset_ComboBox()
   > AddOption(self=@n24, Option=@n23)
   > SetSelectedOption!(execute, self:"Searchable Combo Box (String) Object Reference"=@n24, Option:String=@n23) -> (then)
   > Updated Asset Widget() {is=title K2Node_CallFunction}
   RemoveChild(execute=@n21.then_1, self=Get CanvasPanel(), Content=@n18.Modal Dialog Widget)
   SetVisibility(
    execute=@n20.else
    self=Get ErrorBorder!(self:EUW Asset Browser Modal Object Reference=@n18.Modal Dialog Widget) -> (ErrorBorder:Border Object Reference)
    InVisibility=Select(Option 0=Visible, Option 1=Collapsed, Index=@n19)
   )
   > Set Text(
    Text=@n19.Text
    self=Get ErrorTextBlock!(self:EUW Asset Browser Modal Object Reference=@n18.Modal Dialog Widget) -> (ErrorTextBlock:Text Object Reference)
   )
   Event Construct()
   > Clear(TargetArray:Array of BP Asset Browser Module Object References=Get Modules())
   > n25 = GetDerivedClassNames!(
    execute
    self:Asset Registry Interface=GetAssetRegistry()
    ClassNames:Array of Top Level Asset Path Structures=Make Array("[0]":Top Level Asset Path Structure=GetSoftClassTopLevelAssetPath!(SoftClassReference:Object Soft Class Reference=/AnimInEdTools/AssetBrowser/Module/Base/BP_AssetBrowserModule.BP_AssetBrowserModule_C) -> (ReturnValue:Top Level Asset Path Structure)) -> (Array:Array of Top Level Asset Path Structures)
    ExcludedClassNames:Set of Top Level Asset Path Structures=Make Set!("[0]":Top Level Asset Path Structure) -> (Set:Set of Top Level Asset Path Structures) {is=title K2Node_MakeSet}
   ) -> (then, OutDerivedClassNames:Set of Top Level Asset Path Structures)
   > n26 = To Array!(execute, A:Set of Top Level Asset Path Structures=@n25) -> (then, Result:Array of Top Level Asset Path Structures)
   > n27 = For Each Loop!(Exec, Array:Array of Top Level Asset Path Structures=@n26) -> (LoopBody, Array Element_PathString:String, Array Index, Completed)
   Update Combo Box!(
    execute=@n27.Completed
    self
    Combo Box:"Searchable Combo Box (String) Object Reference"=Get Module_ComboBox() -> (Module_ComboBox:"Searchable Combo Box (String) Object Reference")
    Items:Array of Strings=Get Module Names!(Modules:Array of BP Asset Browser Module Object References=Get Modules()) -> (Names:Array of Strings) {is=title K2Node_CallFunction}
    Default Item:String
    Add Blank:bool=false
   ) -> (then)
   n28 = LoadClassAsset_Blocking(execute=@n27.LoopBody, AssetClass=To Soft Class Reference(SoftClassPath=MakeSoftClassPath(PathString=@n27)))
   > n29 = Construct Object from Class(Class=@n28) -> (ReturnValue:Object Reference)
   > n30 = Cast To BP_AssetBrowserModule!(execute, Object=@n29) -> (then, CastFailed, AsBP Asset Browser Module:BP Asset Browser Module Object Reference) {is=title K2Node_DynamicCast}
   > n31 = Branch(Condition=Get Abstract!(self:BP Asset Browser Module Object Reference=@n30) -> (Abstract:bool))
   Add(
    execute=@n31.else
    TargetArray:Array of BP Asset Browser Module Object References=Get Modules()
    NewItem:"BP Asset Browser Module Object Reference (by ref)"=@n30
   ) {is=title K2Node_CallArrayFunction}
   Create Dialog Cancel Button Clicked!(execute=@n9.then, self, "Modal  Widget":Widget Object Reference=@n11) -> (then) {is=title K2Node_CallFunction}
   Create Dialog Create Clicked!(execute=@n10.then, self, Create Widget:Widget Object Reference=@n6, Modal Dialog Widget:EUW Asset Browser Modal Object Reference=@n11) -> (then) {is=title K2Node_CallFunction}
  ]}
 Find Module by Name {nodes [
   n1 = Find Module by Name!() -> (then, Modules:Array of BP Asset Browser Module Object References, Name:String) {is=title K2Node_FunctionEntry}
   > n2 = For Each Loop with Break!(Exec:Exec, Array:Array of BP Asset Browser Module Object References=@n1, Break:Exec) -> (LoopBody:Exec, Array Element:BP Asset Browser Module Object Reference, Array Index:int, Completed:Exec)
   Return Node!(execute=@n2.Completed, Module:BP Asset Browser Module Object Reference)
   > Branch(Condition="Equal, Case Insensitive (String)"(A=Get Name(self:BP Asset Browser Module Object Reference=@n2), B=@n1.Name))
   > Return Node(Module=@n2)
  ]}
 Get Current Module's Base Dir {nodes [
   Get Current Module's Base Dir() {is=title K2Node_FunctionEntry}
   > n1 = Branch(Condition=IsValid(Object=Get Current Module()))
   > Return Node!(
    execute
    Base Dir:String=Get Default Base Dir!(self:BP Asset Browser Module Object Reference=Get Current Module()) -> (Default Base Dir:String)
   )
   Return Node(execute=@n1.else)
  ]}
 Get Current Module's Editor Plugin {nodes [
   Get Current Module's Editor Plugin() {is=title K2Node_FunctionEntry}
   > n1 = Branch(Condition=IsValid(Object=Get Current Module()))
   n2 = Get Current Module()
   > n3 = Branch(Condition=IsEmpty(InString=Get Default Editor Plugin!(self:BP Asset Browser Module Object Reference=@n2) -> (Default Editor Plugin:String)))
   > Return Node!(execute, Editor Plugin:String=Get Current Module's Plugin!() -> (Plugin:String) {is=title K2Node_CallFunction})
   Return Node(execute=@n3.else, Editor Plugin=Get Default Editor Plugin(self=Get Current Module()))
   Return Node(execute=@n1.else)
  ]}
 Get Current Module's Plugin {nodes [
   Get Current Module's Plugin() {is=title K2Node_FunctionEntry}
   > n1 = Branch(Condition=IsValid(Object=Get Current Module()))
   > Return Node!(execute, Plugin:String=Get Default Plugin!(self:BP Asset Browser Module Object Reference=Get Current Module()) -> (Default Plugin:String))
   Return Node(execute=@n1.else)
  ]}
 Get Module Names {nodes [
   n1 = Get Module Names!() -> (then, Modules:Array of BP Asset Browser Module Object References) {is=title K2Node_FunctionEntry}
   > Clear(TargetArray:Array of Strings=Get Local Names() -> (Local Names:Array of Strings))
   > n2 = For Each Loop!(Exec, Array:Array of BP Asset Browser Module Object References=@n1) -> (LoopBody, Array Element:BP Asset Browser Module Object Reference, Array Index, Completed)
   Return Node!(execute=@n2.Completed, Names:Array of Strings=Get Local Names())
   > Add(TargetArray:Array of Strings=Get Local Names(), NewItem:"String (by ref)"=Get Name(self=@n2))
  ]}
 Get Selected Module {nodes [
   Get Selected Module() {is=title K2Node_FunctionEntry}
   n1 = Get Module_ComboBox()
   > Return Node!(
    execute
    Module:BP Asset Browser Module Object Reference=Find Module by Name(Modules=Get Modules(), Name=GetSelectedOption!(self:"Searchable Combo Box (String) Object Reference"=@n1) -> (ReturnValue:String)) {is=title K2Node_CallFunction}
   )
  ]}
 Update Combo Box {nodes [
   n1 = Update Combo Box!() -> (then, Combo Box:"Searchable Combo Box (String) Object Reference", Items:Array of Strings, Default Item:String, Add Blank:bool) {is=title K2Node_FunctionEntry}
   > n2 = Set Current Index(Current Index:int=GetSelectedIndex!(self:"Searchable Combo Box (String) Object Reference"=@n1) -> (ReturnValue:int=0)) -> (Output_Get:int=0)
   > Set Selected Item(Selected Item:String=GetOptionAtIndex!(
     self:"Searchable Combo Box (String) Object Reference"=Get Combo Box() -> (Combo Box:"Searchable Combo Box (String) Object Reference")
     Index:int=@n2
    ) -> (ReturnValue:String)) -> (Output_Get:String)
   n3 = Get Combo Box()
   > n4 = For Loop(LastIndex=int - int(A=GetOptionCount!(self:"Searchable Combo Box (String) Object Reference"=@n3) -> (ReturnValue:int=0)))
   n5 = Branch(
    execute=@n4.Completed
    Condition=Identical!(ArrayA:Array of Strings=Get Items() -> (Items:Array of Strings), ArrayB:Array of Strings=Get Old Items() -> (Old Items:Array of Strings)) -> (ReturnValue:bool) {is=title K2Node_CallArrayFunction}
   )
   ClearSelection(execute=@n5.else, self:"Searchable Combo Box (String) Object Reference"=Get Combo Box())
   > n6 = Sequence!(execute) -> (then_0, then_1, then_2:Exec, then_3:Exec)
   > Branch(Condition=Get Add Blank() -> (Add Blank:bool))
   n7 = Get Combo Box()
   > AddOption(self=@n7)
   n8 = Get Default Item() -> (Default Item:String)
   n9 = Branch(execute=@n6.then_1, Condition=IsEmpty(InString=@n8))
   AddOption(execute=@n9.else, self=@n7, Option=@n8)
   n10 = For Each Loop(Exec=@n6.then_2, Array:Array of Strings=Get Items()) -> (Array Element:String)
   > AddOption(self=@n7, Option=@n10)
   n11 = Get Combo Box()
   n12 = Get Current Index() -> (Current Index:int=0)
   n13 = Branch(
    execute=@n6.then_3
    Condition="Equal Exactly (String)"(A=GetOptionAtIndex(self=@n11, Index=@n12), B=Get Selected Item() -> (Selected Item:String))
   )
   > SetSelectedIndex!(execute, self:"Searchable Combo Box (String) Object Reference"=@n7, Index:int=Get Current Index() -> (Current Index=0)) -> (then)
   ClearSelection(execute=@n13.else, self=@n11)
   n14 = GetOptionAtIndex(self=Get Combo Box(), Index=@n4)
   n15 = Branch(execute=@n4.LoopBody, Condition="Equal Exactly (String)"(A=@n14, B=Get Default Item()))
   Add(execute=@n15.else, TargetArray=Get Old Items(), NewItem=@n14)
  ]}
]
variables [
 Current Module {type=BP_AssetBrowserModule_C, default=null}
 Module Widget {type=EditorUtilityWidget, default=null}
 Modules {type="array<BP_AssetBrowserModule_C>", default []}
]
widgets {
 info {rootWidgetClass=&CanvasPanel, widgetCount=9}
 widgets [
  CanvasPanel {widgetClassPath=&CanvasPanel, bIsVariable=true, widget=CanvasPanel}
  VerticalBox_0 {
   widgetClassPath=&VerticalBox
   parent=CanvasPanel
   widget=VerticalBox_0
   slot {class=CanvasPanelSlot, properties {layoutData {alignment=<0 0>, anchors {maximum=<1 1>, minimum=<0 0>}, offsets=<ltrb 0 0 0 0>}}}
  }
  HorizontalBox_0 {widgetClassPath=&HorizontalBox, parent=VerticalBox_0, widget=HorizontalBox_0, slot {class=VerticalBoxSlot}}
  SizeBox_314 {
   widgetClassPath=&SizeBox
   parent=HorizontalBox_0
   widget=SizeBox_314
   properties {bOverride_MinDesiredWidth=true, minDesiredWidth=180}
   slot {class=HorizontalBoxSlot}
  }
  Module_ComboBox {
   widgetClassPath=/Script/FortniteEditor.SearchableStringComboBox
   bIsVariable=true
   parent=SizeBox_314
   widget=Module_ComboBox
   properties {
    foregroundColor {colorUseRule=UseColor_Specified, specifiedColor=<rgba 0.5271 0.5271 0.5271 1>}
    itemStyle {
     bUseParentRowBrush=false
     activeBrush {
      bIsDynamicallyLoaded=false
      drawAs=Image
      imageType=NoImage
      mirroring=NoMirror
      resourceName=None
      resourceObject=null
      tiling=NoTile
      imageSize=<0 0>
      margin=<ltrb 0 0 0 0>
      outlineSettings {
       bUseBrushTransparency=false
       roundingType=HalfHeightRadius
       width=0
       color {colorUseRule=UseColor_Specified, specifiedColor=<rgba 0 0 0 0>}
       cornerRadii=<0 0 0 0>
      }
      tintColor {colorUseRule=UseColor_Specified, specifiedColor=<rgba 0 0.162 0.7454 1>}
      uVRegion {bIsValid=false, max=<0 0>, min=<0 0>}
     }
     activeHighlightedBrush {
      bIsDynamicallyLoaded=false
      drawAs=Image
      imageType=NoImage
      mirroring=NoMirror
      resourceName=None
      resourceObject=null
      tiling=NoTile
      imageSize=<0 0>
      margin=<ltrb 0 0 0 0>
      outlineSettings {
       bUseBrushTransparency=false
       roundingType=HalfHeightRadius
       width=0
       color {colorUseRule=UseColor_Specified, specifiedColor=<rgba 0 0 0 0>}
       cornerRadii=<0 0 0 0>
      }
      tintColor {colorUseRule=UseColor_Specified, specifiedColor=<rgba 0.0252 0.0319 0.0423 1>}
      uVRegion {bIsValid=false, max=<0 0>, min=<0 0>}
     }
     activeHoveredBrush {
      bIsDynamicallyLoaded=false
      drawAs=Image
      imageType=NoImage
      mirroring=NoMirror
      resourceName=None
      resourceObject=null
      tiling=NoTile
      imageSize=<0 0>
      margin=<ltrb 0 0 0 0>
      outlineSettings {
       bUseBrushTransparency=false
       roundingType=HalfHeightRadius
       width=0
       color {colorUseRule=UseColor_Specified, specifiedColor=<rgba 0 0 0 0>}
       cornerRadii=<0 0 0 0>
      }
      tintColor {colorUseRule=UseColor_Specified, specifiedColor=<rgba 0 0.162 0.7454 1>}
      uVRegion {bIsValid=false, max=<0 0>, min=<0 0>}
     }
     dropIndicator_Above {
      bIsDynamicallyLoaded=false
      drawAs=Box
      imageType=FullColor
      mirroring=NoMirror
      resourceName=../../../Engine/Content/Slate/Common/DropZoneIndicator_Above.png
      resourceObject=null
      tiling=NoTile
      imageSize=<0 0>
      margin=<ltrb 0.625 0.625 0 0>
      outlineSettings {
       bUseBrushTransparency=false
       roundingType=HalfHeightRadius
       width=0
       color {colorUseRule=UseColor_Specified, specifiedColor=<rgba 0 0 0 0>}
       cornerRadii=<0 0 0 0>
      }
      tintColor {colorUseRule=UseColor_Specified, specifiedColor=<rgba 0.0091 0.3515 0.9647 1>}
      uVRegion {bIsValid=false, max=<0 0>, min=<0 0>}
     }
     dropIndicator_Below {
      bIsDynamicallyLoaded=false
      drawAs=Box
      imageType=FullColor
      mirroring=NoMirror
      resourceName=../../../Engine/Content/Slate/Common/DropZoneIndicator_Below.png
      resourceObject=null
      tiling=NoTile
      imageSize=<0 0>
      margin=<ltrb 0.625 0 0 0.625>
      outlineSettings {
       bUseBrushTransparency=false
       roundingType=HalfHeightRadius
       width=0
       color {colorUseRule=UseColor_Specified, specifiedColor=<rgba 0 0 0 0>}
       cornerRadii=<0 0 0 0>
      }
      tintColor {colorUseRule=UseColor_Specified, specifiedColor=<rgba 0.0091 0.3515 0.9647 1>}
      uVRegion {bIsValid=false, max=<0 0>, min=<0 0>}
     }
     dropIndicator_Onto {
      bIsDynamicallyLoaded=false
      drawAs=Box
      imageType=FullColor
      mirroring=NoMirror
      resourceName=../../../Engine/Content/Slate/Common/DropZoneIndicator_Onto.png
      resourceObject=null
      tiling=NoTile
      imageSize=<0 0>
      margin=<ltrb 0.25 0.25 0.25 0.25>
      outlineSettings {
       bUseBrushTransparency=false
       roundingType=HalfHeightRadius
       width=0
       color {colorUseRule=UseColor_Specified, specifiedColor=<rgba 0 0 0 0>}
       cornerRadii=<0 0 0 0>
      }
      tintColor {colorUseRule=UseColor_Specified, specifiedColor=<rgba 0.0091 0.3515 0.9647 1>}
      uVRegion {bIsValid=false, max=<0 0>, min=<0 0>}
     }
     evenRowBackgroundBrush {
      bIsDynamicallyLoaded=false
      drawAs=Image
      imageType=NoImage
      mirroring=NoMirror
      resourceName=None
      resourceObject=null
      tiling=NoTile
      imageSize=<0 0>
      margin=<ltrb 0 0 0 0>
      outlineSettings {
       bUseBrushTransparency=false
       roundingType=HalfHeightRadius
       width=0
       color {colorUseRule=UseColor_Specified, specifiedColor=<rgba 0 0 0 0>}
       cornerRadii=<0 0 0 0>
      }
      tintColor {colorUseRule=UseColor_Specified, specifiedColor=<rgba 0.0103 0.0103 0.0103 1>}
      uVRegion {bIsValid=false, max=<0 0>, min=<0 0>}
     }
     evenRowBackgroundHoveredBrush {
      bIsDynamicallyLoaded=false
      drawAs=Image
      imageType=NoImage
      mirroring=NoMirror
      resourceName=None
      resourceObject=null
      tiling=NoTile
      imageSize=<0 0>
      margin=<ltrb 0 0 0 0>
      outlineSettings {
       bUseBrushTransparency=false
       roundingType=HalfHeightRadius
       width=0
       color {colorUseRule=UseColor_Specified, specifiedColor=<rgba 0 0 0 0>}
       cornerRadii=<0 0 0 0>
      }
      tintColor {colorUseRule=UseColor_Specified, specifiedColor=<rgba 0.0176 0.0176 0.0176 1>}
      uVRegion {bIsValid=false, max=<0 0>, min=<0 0>}
     }
     inactiveBrush {
      bIsDynamicallyLoaded=false
      drawAs=Image
      imageType=NoImage
      mirroring=NoMirror
      resourceName=None
      resourceObject=null
      tiling=NoTile
      imageSize=<0 0>
      margin=<ltrb 0 0 0 0>
      outlineSettings {
       bUseBrushTransparency=false
       roundingType=HalfHeightRadius
       width=0
       color {colorUseRule=UseColor_Specified, specifiedColor=<rgba 0 0 0 0>}
       cornerRadii=<0 0 0 0>
      }
      tintColor {colorUseRule=UseColor_Specified, specifiedColor=<rgba 0.0513 0.0953 0.159 1>}
      uVRegion {bIsValid=false, max=<0 0>, min=<0 0>}
     }
     inactiveHighlightedBrush {
      bIsDynamicallyLoaded=false
      drawAs=Image
      imageType=NoImage
      mirroring=NoMirror
      resourceName=None
      resourceObject=null
      tiling=NoTile
      imageSize=<0 0>
      margin=<ltrb 0 0 0 0>
      outlineSettings {
       bUseBrushTransparency=false
       roundingType=HalfHeightRadius
       width=0
       color {colorUseRule=UseColor_Specified, specifiedColor=<rgba 0 0 0 0>}
       cornerRadii=<0 0 0 0>
      }
      tintColor {colorUseRule=UseColor_Specified, specifiedColor=<rgba 0.0252 0.0319 0.0423 1>}
      uVRegion {bIsValid=false, max=<0 0>, min=<0 0>}
     }
     inactiveHoveredBrush {
      bIsDynamicallyLoaded=false
      drawAs=Image
      imageType=NoImage
      mirroring=NoMirror
      resourceName=None
      resourceObject=null
      tiling=NoTile
      imageSize=<0 0>
      margin=<ltrb 0 0 0 0>
      outlineSettings {
       bUseBrushTransparency=false
       roundingType=HalfHeightRadius
       width=0
       color {colorUseRule=UseColor_Specified, specifiedColor=<rgba 0 0 0 0>}
       cornerRadii=<0 0 0 0>
      }
      tintColor {colorUseRule=UseColor_Specified, specifiedColor=<rgba 0.0513 0.0953 0.159 1>}
      uVRegion {bIsValid=false, max=<0 0>, min=<0 0>}
     }
     oddRowBackgroundBrush {
      bIsDynamicallyLoaded=false
      drawAs=Image
      imageType=NoImage
      mirroring=NoMirror
      resourceName=None
      resourceObject=null
      tiling=NoTile
      imageSize=<0 0>
      margin=<ltrb 0 0 0 0>
      outlineSettings {
       bUseBrushTransparency=false
       roundingType=HalfHeightRadius
       width=0
       color {colorUseRule=UseColor_Specified, specifiedColor=<rgba 0 0 0 0>}
       cornerRadii=<0 0 0 0>
      }
      tintColor {colorUseRule=UseColor_Specified, specifiedColor=<rgba 0.0103 0.0103 0.0103 1>}
      uVRegion {bIsValid=false, max=<0 0>, min=<0 0>}
     }
     oddRowBackgroundHoveredBrush {
      bIsDynamicallyLoaded=false
      drawAs=Image
      imageType=NoImage
      mirroring=NoMirror
      resourceName=None
      resourceObject=null
      tiling=NoTile
      imageSize=<0 0>
      margin=<ltrb 0 0 0 0>
      outlineSettings {
       bUseBrushTransparency=false
       roundingType=HalfHeightRadius
       width=0
       color {colorUseRule=UseColor_Specified, specifiedColor=<rgba 0 0 0 0>}
       cornerRadii=<0 0 0 0>
      }
      tintColor {colorUseRule=UseColor_Specified, specifiedColor=<rgba 0.0176 0.0176 0.0176 1>}
      uVRegion {bIsValid=false, max=<0 0>, min=<0 0>}
     }
     parentRowBackgroundBrush {
      bIsDynamicallyLoaded=false
      drawAs=Image
      imageType=NoImage
      mirroring=NoMirror
      resourceName=None
      resourceObject=null
      tiling=NoTile
      imageSize=<32 32>
      margin=<ltrb 0 0 0 0>
      outlineSettings {
       bUseBrushTransparency=false
       roundingType=HalfHeightRadius
       width=0
       color {colorUseRule=UseColor_Specified, specifiedColor=<rgba 0 0 0 0>}
       cornerRadii=<0 0 0 0>
      }
      tintColor {colorUseRule=UseColor_Specified, specifiedColor=<rgba 1 1 1 1>}
      uVRegion {bIsValid=false, max=<0 0>, min=<0 0>}
     }
     parentRowBackgroundHoveredBrush {
      bIsDynamicallyLoaded=false
      drawAs=Image
      imageType=NoImage
      mirroring=NoMirror
      resourceName=None
      resourceObject=null
      tiling=NoTile
      imageSize=<32 32>
      margin=<ltrb 0 0 0 0>
      outlineSettings {
       bUseBrushTransparency=false
       roundingType=HalfHeightRadius
       width=0
       color {colorUseRule=UseColor_Specified, specifiedColor=<rgba 0 0 0 0>}
       cornerRadii=<0 0 0 0>
      }
      tintColor {colorUseRule=UseColor_Specified, specifiedColor=<rgba 1 1 1 1>}
      uVRegion {bIsValid=false, max=<0 0>, min=<0 0>}
     }
     selectedTextColor {colorUseRule=UseColor_Specified, specifiedColor=<rgba 1 1 1 1>}
     selectorFocusedBrush {
      bIsDynamicallyLoaded=false
      drawAs=Border
      imageType=FullColor
      mirroring=NoMirror
      resourceName=../../../Engine/Content/Slate/Common/Selector.png
      resourceObject=null
      tiling=Both
      imageSize=<0 0>
      margin=<ltrb 0.25 0.25 0.25 0.25>
      outlineSettings {
       bUseBrushTransparency=false
       roundingType=HalfHeightRadius
       width=0
       color {colorUseRule=UseColor_Specified, specifiedColor=<rgba 0 0 0 0>}
       cornerRadii=<0 0 0 0>
      }
      tintColor {colorUseRule=UseColor_Specified, specifiedColor=<rgba 0.701 0.225 0.003 1>}
      uVRegion {bIsValid=false, max=<0 0>, min=<0 0>}
     }
     textColor {colorUseRule=UseColor_Specified, specifiedColor=<rgba 0.5271 0.5271 0.5271 1>}
    }
    widgetStyle {
     comboButtonStyle {
      downArrowAlign=VAlign_Center
      buttonStyle {
       clickedSlateSound {resourceObject=null}
       disabled {
        bIsDynamicallyLoaded=false
        drawAs=RoundedBox
        imageType=NoImage
        mirroring=NoMirror
        resourceName=None
        resourceObject=null
        tiling=NoTile
        imageSize=<0 0>
        margin=<ltrb 0 0 0 0>
        outlineSettings {
         bUseBrushTransparency=false
         roundingType=FixedRadius
         width=1
         color {colorUseRule=UseColor_ColorTable, specifiedColor=<rgba 0 0 0 0>}
         cornerRadii=<4 4 4 4>
        }
        tintColor {colorUseRule=UseColor_Specified, specifiedColor=<rgba 0.0395 0.0395 0.0395 1>}
        uVRegion {bIsValid=false, max=<0 0>, min=<0 0>}
       }
       disabledForeground {colorUseRule=UseColor_ColorTable, specifiedColor=<rgba 0 0 0 0>}
       hovered {
        bIsDynamicallyLoaded=false
        drawAs=RoundedBox
        imageType=NoImage
        mirroring=NoMirror
        resourceName=None
        resourceObject=null
        tiling=NoTile
        imageSize=<0 0>
        margin=<ltrb 0 0 0 0>
        outlineSettings {
         bUseBrushTransparency=false
         roundingType=FixedRadius
         width=1
         color {colorUseRule=UseColor_ColorTable, specifiedColor=<rgba 0 0 0 0>}
         cornerRadii=<4 4 4 4>
        }
        tintColor {colorUseRule=UseColor_Specified, specifiedColor=<rgba 0.0048 0.0048 0.0048 1>}
        uVRegion {bIsValid=false, max=<0 0>, min=<0 0>}
       }
       hoveredForeground {colorUseRule=UseColor_ColorTable, specifiedColor=<rgba 0 0 0 0>}
       hoveredSlateSound {resourceObject=null}
       normal {
        bIsDynamicallyLoaded=false
        drawAs=RoundedBox
        imageType=NoImage
        mirroring=NoMirror
        resourceName=None
        resourceObject=null
        tiling=NoTile
        imageSize=<0 0>
        margin=<ltrb 0 0 0 0>
        outlineSettings {
         bUseBrushTransparency=false
         roundingType=FixedRadius
         width=1
         color {colorUseRule=UseColor_ColorTable, specifiedColor=<rgba 0 0 0 0>}
         cornerRadii=<4 4 4 4>
        }
        tintColor {colorUseRule=UseColor_Specified, specifiedColor=<rgba 0.0048 0.0048 0.0048 1>}
        uVRegion {bIsValid=false, max=<0 0>, min=<0 0>}
       }
       normalForeground {colorUseRule=UseColor_ColorTable, specifiedColor=<rgba 0 0 0 0>}
       normalPadding=<ltrb 8 1 8 1>
       pressed {
        bIsDynamicallyLoaded=false
        drawAs=RoundedBox
        imageType=NoImage
        mirroring=NoMirror
        resourceName=None
        resourceObject=null
        tiling=NoTile
        imageSize=<0 0>
        margin=<ltrb 0 0 0 0>
        outlineSettings {
         bUseBrushTransparency=false
         roundingType=FixedRadius
         width=1
         color {colorUseRule=UseColor_ColorTable, specifiedColor=<rgba 0 0 0 0>}
         cornerRadii=<4 4 4 4>
        }
        tintColor {colorUseRule=UseColor_Specified, specifiedColor=<rgba 0.0048 0.0048 0.0048 1>}
        uVRegion {bIsValid=false, max=<0 0>, min=<0 0>}
       }
       pressedForeground {colorUseRule=UseColor_ColorTable, specifiedColor=<rgba 0 0 0 0>}
       pressedPadding=<ltrb 8 2 8 0>
       pressedSlateSound {resourceObject=null}
      }
      contentPadding=<ltrb 0 0 0 0>
      downArrowImage {
       bIsDynamicallyLoaded=false
       drawAs=Image
       imageType=Vector
       mirroring=NoMirror
       resourceName=../../../Engine/Content/Slate/Starship/CoreWidgets/ComboBox/wide-chevron-down.svg
       resourceObject=null
       tiling=NoTile
       imageSize=<16 16>
       margin=<ltrb 0 0 0 0>
       outlineSettings {
        bUseBrushTransparency=false
        roundingType=HalfHeightRadius
        width=0
        color {colorUseRule=UseColor_Specified, specifiedColor=<rgba 0 0 0 0>}
        cornerRadii=<0 0 0 0>
       }
       tintColor {colorUseRule=UseColor_Specified, specifiedColor=<rgba 1 1 1 1>}
       uVRegion {bIsValid=false, max=<0 0>, min=<0 0>}
      }
      downArrowPadding=<ltrb 2 2 2 2>
      menuBorderBrush {
       bIsDynamicallyLoaded=false
       drawAs=RoundedBox
       imageType=NoImage
       mirroring=NoMirror
       resourceName=None
       resourceObject=null
       tiling=NoTile
       imageSize=<0 0>
       margin=<ltrb 0 0 0 0>
       outlineSettings {
        bUseBrushTransparency=false
        roundingType=FixedRadius
        width=1
        color {colorUseRule=UseColor_ColorTable, specifiedColor=<rgba 0 0 0 0>}
        cornerRadii=<0 0 0 0>
       }
       tintColor {colorUseRule=UseColor_Specified, specifiedColor=<rgba 0.0395 0.0395 0.0395 1>}
       uVRegion {bIsValid=false, max=<0 0>, min=<0 0>}
      }
      menuBorderPadding=<ltrb 0 0 0 0>
      shadowColorAndOpacity=<rgba 0 0 0 1>
      shadowOffset=<0 0>
     }
     contentPadding=<ltrb 0 0 0 0>
     menuRowPadding=<ltrb 8 3 8 3>
     pressedSlateSound {resourceObject=null}
     selectionChangeSlateSound {resourceObject=null}
    }
   }
   slot {class=SizeBoxSlot}
  }
  Asset_ComboBox {
   widgetClassPath=&SearchableStringComboBox
   bIsVariable=true
   parent=HorizontalBox_0
   widget=Asset_ComboBox
   properties {
    foregroundColor {colorUseRule=UseColor_Specified, specifiedColor=<rgba 0.5271 0.5271 0.5271 1>}
    itemStyle {
     bUseParentRowBrush=false
     activeBrush {
      bIsDynamicallyLoaded=false
      drawAs=Image
      imageType=NoImage
      mirroring=NoMirror
      resourceName=None
      resourceObject=null
      tiling=NoTile
      imageSize=<0 0>
      margin=<ltrb 0 0 0 0>
      outlineSettings {
       bUseBrushTransparency=false
       roundingType=HalfHeightRadius
       width=0
       color {colorUseRule=UseColor_Specified, specifiedColor=<rgba 0 0 0 0>}
       cornerRadii=<0 0 0 0>
      }
      tintColor {colorUseRule=UseColor_Specified, specifiedColor=<rgba 0 0.162 0.7454 1>}
      uVRegion {bIsValid=false, max=<0 0>, min=<0 0>}
     }
     activeHighlightedBrush {
      bIsDynamicallyLoaded=false
      drawAs=Image
      imageType=NoImage
      mirroring=NoMirror
      resourceName=None
      resourceObject=null
      tiling=NoTile
      imageSize=<0 0>
      margin=<ltrb 0 0 0 0>
      outlineSettings {
       bUseBrushTransparency=false
       roundingType=HalfHeightRadius
       width=0
       color {colorUseRule=UseColor_Specified, specifiedColor=<rgba 0 0 0 0>}
       cornerRadii=<0 0 0 0>
      }
      tintColor {colorUseRule=UseColor_Specified, specifiedColor=<rgba 0.0252 0.0319 0.0423 1>}
      uVRegion {bIsValid=false, max=<0 0>, min=<0 0>}
     }
     activeHoveredBrush {
      bIsDynamicallyLoaded=false
      drawAs=Image
      imageType=NoImage
      mirroring=NoMirror
      resourceName=None
      resourceObject=null
      tiling=NoTile
      imageSize=<0 0>
      margin=<ltrb 0 0 0 0>
      outlineSettings {
       bUseBrushTransparency=false
       roundingType=HalfHeightRadius
       width=0
       color {colorUseRule=UseColor_Specified, specifiedColor=<rgba 0 0 0 0>}
       cornerRadii=<0 0 0 0>
      }
      tintColor {colorUseRule=UseColor_Specified, specifiedColor=<rgba 0 0.162 0.7454 1>}
      uVRegion {bIsValid=false, max=<0 0>, min=<0 0>}
     }
     dropIndicator_Above {
      bIsDynamicallyLoaded=false
      drawAs=Box
      imageType=FullColor
      mirroring=NoMirror
      resourceName=../../../Engine/Content/Slate/Common/DropZoneIndicator_Above.png
      resourceObject=null
      tiling=NoTile
      imageSize=<0 0>
      margin=<ltrb 0.625 0.625 0 0>
      outlineSettings {
       bUseBrushTransparency=false
       roundingType=HalfHeightRadius
       width=0
       color {colorUseRule=UseColor_Specified, specifiedColor=<rgba 0 0 0 0>}
       cornerRadii=<0 0 0 0>
      }
      tintColor {colorUseRule=UseColor_Specified, specifiedColor=<rgba 0.0091 0.3515 0.9647 1>}
      uVRegion {bIsValid=false, max=<0 0>, min=<0 0>}
     }
     dropIndicator_Below {
      bIsDynamicallyLoaded=false
      drawAs=Box
      imageType=FullColor
      mirroring=NoMirror
      resourceName=../../../Engine/Content/Slate/Common/DropZoneIndicator_Below.png
      resourceObject=null
      tiling=NoTile
      imageSize=<0 0>
      margin=<ltrb 0.625 0 0 0.625>
      outlineSettings {
       bUseBrushTransparency=false
       roundingType=HalfHeightRadius
       width=0
       color {colorUseRule=UseColor_Specified, specifiedColor=<rgba 0 0 0 0>}
       cornerRadii=<0 0 0 0>
      }
      tintColor {colorUseRule=UseColor_Specified, specifiedColor=<rgba 0.0091 0.3515 0.9647 1>}
      uVRegion {bIsValid=false, max=<0 0>, min=<0 0>}
     }
     dropIndicator_Onto {
      bIsDynamicallyLoaded=false
      drawAs=Box
      imageType=FullColor
      mirroring=NoMirror
      resourceName=../../../Engine/Content/Slate/Common/DropZoneIndicator_Onto.png
      resourceObject=null
      tiling=NoTile
      imageSize=<0 0>
      margin=<ltrb 0.25 0.25 0.25 0.25>
      outlineSettings {
       bUseBrushTransparency=false
       roundingType=HalfHeightRadius
       width=0
       color {colorUseRule=UseColor_Specified, specifiedColor=<rgba 0 0 0 0>}
       cornerRadii=<0 0 0 0>
      }
      tintColor {colorUseRule=UseColor_Specified, specifiedColor=<rgba 0.0091 0.3515 0.9647 1>}
      uVRegion {bIsValid=false, max=<0 0>, min=<0 0>}
     }
     evenRowBackgroundBrush {
      bIsDynamicallyLoaded=false
      drawAs=Image
      imageType=NoImage
      mirroring=NoMirror
      resourceName=None
      resourceObject=null
      tiling=NoTile
      imageSize=<0 0>
      margin=<ltrb 0 0 0 0>
      outlineSettings {
       bUseBrushTransparency=false
       roundingType=HalfHeightRadius
       width=0
       color {colorUseRule=UseColor_Specified, specifiedColor=<rgba 0 0 0 0>}
       cornerRadii=<0 0 0 0>
      }
      tintColor {colorUseRule=UseColor_Specified, specifiedColor=<rgba 0.0103 0.0103 0.0103 1>}
      uVRegion {bIsValid=false, max=<0 0>, min=<0 0>}
     }
     evenRowBackgroundHoveredBrush {
      bIsDynamicallyLoaded=false
      drawAs=Image
      imageType=NoImage
      mirroring=NoMirror
      resourceName=None
      resourceObject=null
      tiling=NoTile
      imageSize=<0 0>
      margin=<ltrb 0 0 0 0>
      outlineSettings {
       bUseBrushTransparency=false
       roundingType=HalfHeightRadius
       width=0
       color {colorUseRule=UseColor_Specified, specifiedColor=<rgba 0 0 0 0>}
       cornerRadii=<0 0 0 0>
      }
      tintColor {colorUseRule=UseColor_Specified, specifiedColor=<rgba 0.0176 0.0176 0.0176 1>}
      uVRegion {bIsValid=false, max=<0 0>, min=<0 0>}
     }
     inactiveBrush {
      bIsDynamicallyLoaded=false
      drawAs=Image
      imageType=NoImage
      mirroring=NoMirror
      resourceName=None
      resourceObject=null
      tiling=NoTile
      imageSize=<0 0>
      margin=<ltrb 0 0 0 0>
      outlineSettings {
       bUseBrushTransparency=false
       roundingType=HalfHeightRadius
       width=0
       color {colorUseRule=UseColor_Specified, specifiedColor=<rgba 0 0 0 0>}
       cornerRadii=<0 0 0 0>
      }
      tintColor {colorUseRule=UseColor_Specified, specifiedColor=<rgba 0.0513 0.0953 0.159 1>}
      uVRegion {bIsValid=false, max=<0 0>, min=<0 0>}
     }
     inactiveHighlightedBrush {
      bIsDynamicallyLoaded=false
      drawAs=Image
      imageType=NoImage
      mirroring=NoMirror
      resourceName=None
      resourceObject=null
      tiling=NoTile
      imageSize=<0 0>
      margin=<ltrb 0 0 0 0>
      outlineSettings {
       bUseBrushTransparency=false
       roundingType=HalfHeightRadius
       width=0
       color {colorUseRule=UseColor_Specified, specifiedColor=<rgba 0 0 0 0>}
       cornerRadii=<0 0 0 0>
      }
      tintColor {colorUseRule=UseColor_Specified, specifiedColor=<rgba 0.0252 0.0319 0.0423 1>}
      uVRegion {bIsValid=false, max=<0 0>, min=<0 0>}
     }
     inactiveHoveredBrush {
      bIsDynamicallyLoaded=false
      drawAs=Image
      imageType=NoImage
      mirroring=NoMirror
      resourceName=None
      resourceObject=null
      tiling=NoTile
      imageSize=<0 0>
      margin=<ltrb 0 0 0 0>
      outlineSettings {
       bUseBrushTransparency=false
       roundingType=HalfHeightRadius
       width=0
       color {colorUseRule=UseColor_Specified, specifiedColor=<rgba 0 0 0 0>}
       cornerRadii=<0 0 0 0>
      }
      tintColor {colorUseRule=UseColor_Specified, specifiedColor=<rgba 0.0513 0.0953 0.159 1>}
      uVRegion {bIsValid=false, max=<0 0>, min=<0 0>}
     }
     oddRowBackgroundBrush {
      bIsDynamicallyLoaded=false
      drawAs=Image
      imageType=NoImage
      mirroring=NoMirror
      resourceName=None
      resourceObject=null
      tiling=NoTile
      imageSize=<0 0>
      margin=<ltrb 0 0 0 0>
      outlineSettings {
       bUseBrushTransparency=false
       roundingType=HalfHeightRadius
       width=0
       color {colorUseRule=UseColor_Specified, specifiedColor=<rgba 0 0 0 0>}
       cornerRadii=<0 0 0 0>
      }
      tintColor {colorUseRule=UseColor_Specified, specifiedColor=<rgba 0.0103 0.0103 0.0103 1>}
      uVRegion {bIsValid=false, max=<0 0>, min=<0 0>}
     }
     oddRowBackgroundHoveredBrush {
      bIsDynamicallyLoaded=false
      drawAs=Image
      imageType=NoImage
      mirroring=NoMirror
      resourceName=None
      resourceObject=null
      tiling=NoTile
      imageSize=<0 0>
      margin=<ltrb 0 0 0 0>
      outlineSettings {
       bUseBrushTransparency=false
       roundingType=HalfHeightRadius
       width=0
       color {colorUseRule=UseColor_Specified, specifiedColor=<rgba 0 0 0 0>}
       cornerRadii=<0 0 0 0>
      }
      tintColor {colorUseRule=UseColor_Specified, specifiedColor=<rgba 0.0176 0.0176 0.0176 1>}
      uVRegion {bIsValid=false, max=<0 0>, min=<0 0>}
     }
     parentRowBackgroundBrush {
      bIsDynamicallyLoaded=false
      drawAs=Image
      imageType=NoImage
      mirroring=NoMirror
      resourceName=None
      resourceObject=null
      tiling=NoTile
      imageSize=<32 32>
      margin=<ltrb 0 0 0 0>
      outlineSettings {
       bUseBrushTransparency=false
       roundingType=HalfHeightRadius
       width=0
       color {colorUseRule=UseColor_Specified, specifiedColor=<rgba 0 0 0 0>}
       cornerRadii=<0 0 0 0>
      }
      tintColor {colorUseRule=UseColor_Specified, specifiedColor=<rgba 1 1 1 1>}
      uVRegion {bIsValid=false, max=<0 0>, min=<0 0>}
     }
     parentRowBackgroundHoveredBrush {
      bIsDynamicallyLoaded=false
      drawAs=Image
      imageType=NoImage
      mirroring=NoMirror
      resourceName=None
      resourceObject=null
      tiling=NoTile
      imageSize=<32 32>
      margin=<ltrb 0 0 0 0>
      outlineSettings {
       bUseBrushTransparency=false
       roundingType=HalfHeightRadius
       width=0
       color {colorUseRule=UseColor_Specified, specifiedColor=<rgba 0 0 0 0>}
       cornerRadii=<0 0 0 0>
      }
      tintColor {colorUseRule=UseColor_Specified, specifiedColor=<rgba 1 1 1 1>}
      uVRegion {bIsValid=false, max=<0 0>, min=<0 0>}
     }
     selectedTextColor {colorUseRule=UseColor_Specified, specifiedColor=<rgba 1 1 1 1>}
     selectorFocusedBrush {
      bIsDynamicallyLoaded=false
      drawAs=Border
      imageType=FullColor
      mirroring=NoMirror
      resourceName=../../../Engine/Content/Slate/Common/Selector.png
      resourceObject=null
      tiling=Both
      imageSize=<0 0>
      margin=<ltrb 0.25 0.25 0.25 0.25>
      outlineSettings {
       bUseBrushTransparency=false
       roundingType=HalfHeightRadius
       width=0
       color {colorUseRule=UseColor_Specified, specifiedColor=<rgba 0 0 0 0>}
       cornerRadii=<0 0 0 0>
      }
      tintColor {colorUseRule=UseColor_Specified, specifiedColor=<rgba 0.701 0.225 0.003 1>}
      uVRegion {bIsValid=false, max=<0 0>, min=<0 0>}
     }
     textColor {colorUseRule=UseColor_Specified, specifiedColor=<rgba 0.5271 0.5271 0.5271 1>}
    }
    widgetStyle {
     comboButtonStyle {
      downArrowAlign=VAlign_Center
      buttonStyle {
       clickedSlateSound {resourceObject=null}
       disabled {
        bIsDynamicallyLoaded=false
        drawAs=RoundedBox
        imageType=NoImage
        mirroring=NoMirror
        resourceName=None
        resourceObject=null
        tiling=NoTile
        imageSize=<0 0>
        margin=<ltrb 0 0 0 0>
        outlineSettings {
         bUseBrushTransparency=false
         roundingType=FixedRadius
         width=1
         color {colorUseRule=UseColor_ColorTable, specifiedColor=<rgba 0 0 0 0>}
         cornerRadii=<4 4 4 4>
        }
        tintColor {colorUseRule=UseColor_Specified, specifiedColor=<rgba 0.0395 0.0395 0.0395 1>}
        uVRegion {bIsValid=false, max=<0 0>, min=<0 0>}
       }
       disabledForeground {colorUseRule=UseColor_ColorTable, specifiedColor=<rgba 0 0 0 0>}
       hovered {
        bIsDynamicallyLoaded=false
        drawAs=RoundedBox
        imageType=NoImage
        mirroring=NoMirror
        resourceName=None
        resourceObject=null
        tiling=NoTile
        imageSize=<0 0>
        margin=<ltrb 0 0 0 0>
        outlineSettings {
         bUseBrushTransparency=false
         roundingType=FixedRadius
         width=1
         color {colorUseRule=UseColor_ColorTable, specifiedColor=<rgba 0 0 0 0>}
         cornerRadii=<4 4 4 4>
        }
        tintColor {colorUseRule=UseColor_Specified, specifiedColor=<rgba 0.0048 0.0048 0.0048 1>}
        uVRegion {bIsValid=false, max=<0 0>, min=<0 0>}
       }
       hoveredForeground {colorUseRule=UseColor_ColorTable, specifiedColor=<rgba 0 0 0 0>}
       hoveredSlateSound {resourceObject=null}
       normal {
        bIsDynamicallyLoaded=false
        drawAs=RoundedBox
        imageType=NoImage
        mirroring=NoMirror
        resourceName=None
        resourceObject=null
        tiling=NoTile
        imageSize=<0 0>
        margin=<ltrb 0 0 0 0>
        outlineSettings {
         bUseBrushTransparency=false
         roundingType=FixedRadius
         width=1
         color {colorUseRule=UseColor_ColorTable, specifiedColor=<rgba 0 0 0 0>}
         cornerRadii=<4 4 4 4>
        }
        tintColor {colorUseRule=UseColor_Specified, specifiedColor=<rgba 0.0048 0.0048 0.0048 1>}
        uVRegion {bIsValid=false, max=<0 0>, min=<0 0>}
       }
       normalForeground {colorUseRule=UseColor_ColorTable, specifiedColor=<rgba 0 0 0 0>}
       normalPadding=<ltrb 8 1 8 1>
       pressed {
        bIsDynamicallyLoaded=false
        drawAs=RoundedBox
        imageType=NoImage
        mirroring=NoMirror
        resourceName=None
        resourceObject=null
        tiling=NoTile
        imageSize=<0 0>
        margin=<ltrb 0 0 0 0>
        outlineSettings {
         bUseBrushTransparency=false
         roundingType=FixedRadius
         width=1
         color {colorUseRule=UseColor_ColorTable, specifiedColor=<rgba 0 0 0 0>}
         cornerRadii=<4 4 4 4>
        }
        tintColor {colorUseRule=UseColor_Specified, specifiedColor=<rgba 0.0048 0.0048 0.0048 1>}
        uVRegion {bIsValid=false, max=<0 0>, min=<0 0>}
       }
       pressedForeground {colorUseRule=UseColor_ColorTable, specifiedColor=<rgba 0 0 0 0>}
       pressedPadding=<ltrb 8 2 8 0>
       pressedSlateSound {resourceObject=null}
      }
      contentPadding=<ltrb 0 0 0 0>
      downArrowImage {
       bIsDynamicallyLoaded=false
       drawAs=Image
       imageType=Vector
       mirroring=NoMirror
       resourceName=../../../Engine/Content/Slate/Starship/CoreWidgets/ComboBox/wide-chevron-down.svg
       resourceObject=null
       tiling=NoTile
       imageSize=<16 16>
       margin=<ltrb 0 0 0 0>
       outlineSettings {
        bUseBrushTransparency=false
        roundingType=HalfHeightRadius
        width=0
        color {colorUseRule=UseColor_Specified, specifiedColor=<rgba 0 0 0 0>}
        cornerRadii=<0 0 0 0>
       }
       tintColor {colorUseRule=UseColor_Specified, specifiedColor=<rgba 1 1 1 1>}
       uVRegion {bIsValid=false, max=<0 0>, min=<0 0>}
      }
      downArrowPadding=<ltrb 2 2 2 2>
      menuBorderBrush {
       bIsDynamicallyLoaded=false
       drawAs=RoundedBox
       imageType=NoImage
       mirroring=NoMirror
       resourceName=None
       resourceObject=null
       tiling=NoTile
       imageSize=<0 0>
       margin=<ltrb 0 0 0 0>
       outlineSettings {
        bUseBrushTransparency=false
        roundingType=FixedRadius
        width=1
        color {colorUseRule=UseColor_ColorTable, specifiedColor=<rgba 0 0 0 0>}
        cornerRadii=<0 0 0 0>
       }
       tintColor {colorUseRule=UseColor_Specified, specifiedColor=<rgba 0.0395 0.0395 0.0395 1>}
       uVRegion {bIsValid=false, max=<0 0>, min=<0 0>}
      }
      menuBorderPadding=<ltrb 0 0 0 0>
      shadowColorAndOpacity=<rgba 0 0 0 1>
      shadowOffset=<0 0>
     }
     contentPadding=<ltrb 0 0 0 0>
     menuRowPadding=<ltrb 8 3 8 3>
     pressedSlateSound {resourceObject=null}
     selectionChangeSlateSound {resourceObject=null}
    }
   }
   slot {class=HorizontalBoxSlot, properties {size {sizeRule=Fill, value=1}}}
  }
  Add_Button {
   widgetClassPath=&EditorUtilityButton
   parent=HorizontalBox_0
   widget=Add_Button
   properties {visibility=Collapsed}
   slot {class=HorizontalBoxSlot}
  }
  Text_Add {
   widgetClassPath=&TextBlock
   parent=Add_Button
   widget=Text_Add
   properties {
    text=Add
    font {
     bForceMonospaced=false
     bMaterialIsStencil=false
     fontMaterial=null
     fontObject=&Roboto
     letterSpacing=0
     monospacedWidth=1
     size=13.5
     skewAmount=0
     typefaceFontName=Bold
     outlineSettings {
      bApplyOutlineToDropShadows=false
      bMiteredCorners=false
      bSeparateFillAlpha=false
      outlineMaterial=null
      outlineSize=0
      outlineColor=<rgba 0 0 0 1>
     }
    }
   }
   slot {class=ButtonSlot}
  }
  NamedSlot {
   widgetClassPath=&NamedSlot
   bIsVariable=true
   parent=VerticalBox_0
   widget=NamedSlot
   slot {class=VerticalBoxSlot, properties {size {sizeRule=Fill, value=1}}}
  }
 ]
}
