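; Editor function library for the AnimInEdTools asset browser that lists game feature plugins and generates asset names and paths.
; Plugin listing loads each FortGameFeatureData asset, optionally filtering plugin names by virtual path prefix.
; Class prefix and directory maps from BP_AssetBrowser_NameMappings feed name and path builders, which the asset variants wrap.
; Map lookup tries the exact class, then loads key classes and takes the first parent match, otherwise returns empty.
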(fn FilterGameFeaturePlugins
  (params (Virtual%20Path%20Filter :type String))
  (Utilities|Array|Clear (LocalVariables|GetFound%20Plugins))
  (bind _outassetdata (AssetRegistry|GetAssetsbyClass (AssetRegistry|GetAssetRegistry) (Utilities|GetSoftClassTopLevelAssetPath "/Script/FortniteGame.FortGameFeatureData") false))
  (for _array_element _outassetdata
    (bind (_array_element_packagename _array_element_packagepath _array_element_assetname) (Utilities|Struct|BreakAssetData _array_element))
    (bind _returnvalue (Utilities|LoadAssetBlocking (Utilities|ToSoftObjectReference (Utilities|MakeSoftObjectPath (Utilities|String|Append (Utilities|String|ToString(Name) _array_element_packagename) "." (Utilities|String|ToString(Name) _array_element_assetname))))))
    (bind _asgame_feature_data (Utilities|Casting|CastToGameFeatureData _returnvalue))
    (bind _pluginname (GameFeature|GetPluginName _asgame_feature_data))
    (if (Utilities|String|IsEmpty (Variables|GetVirtual%20Path%20Filter))
      (Utilities|Array|Add (LocalVariables|GetFound%20Plugins) _pluginname)
      (else
        (bind _outvirtualpath (EngineScripting|PluginUtilities|GetPluginEditorCustomVirtualPath _pluginname))
        (if (Utilities|String|StartsWith _outvirtualpath (Variables|GetVirtual%20Path%20Filter))
          (Utilities|Array|Add (LocalVariables|GetFound%20Plugins) _pluginname)))))
  (return (LocalVariables|GetFound%20Plugins)))

(fn GenerateAssetNameForAsset
  (params (Base%20Name :type String) (Asset :type Object%20Reference))
  (return (CallFunction|GenerateAssetNameforClass Base%20Name (Utilities|GetClass Asset))))

(fn GenerateAssetNameForClass
  (params (Base%20Name :type String) (Asset%20Class :type Object%20Class%20Reference))
  (return (Utilities|String|Append (Utilities|String|SelectString "" (Utilities|String|Append (CallFunction|GetValueFromClassNameMap Asset%20Class (Variables|/AnimInEdTools/AssetBrowser/Blueprint/BP_AssetBrowser_NameMappings.BP_AssetBrowser_NameMappings_C|GetClass%20Prefix%20Map (LocalVariables|GetClass%20Name%20Mappings))) "_") (Utilities|String|IsEmpty (CallFunction|GetValueFromClassNameMap Asset%20Class (Variables|/AnimInEdTools/AssetBrowser/Blueprint/BP_AssetBrowser_NameMappings.BP_AssetBrowser_NameMappings_C|GetClass%20Prefix%20Map (LocalVariables|GetClass%20Name%20Mappings))))) (Variables|GetBase%20Name))))

(fn GenerateAssetPathForAsset
  (params (Plugin :type String) (Base%20Dir :type String) (Base%20Name :type String) (Asset :type Object%20Reference))
  (bind (_asset_name _package_path _full_path) (CallFunction|GenerateAssetPathforClass Plugin Base%20Dir Base%20Name (Utilities|GetClass Asset)))
  (return _asset_name _package_path _full_path))

(fn GenerateAssetPathForClass
  (params (Plugin :type String) (Base%20Dir :type String) (Base%20Name :type String) (Asset%20Class :type Object%20Class%20Reference))
  (return (CallFunction|GenerateAssetNameforClass Base%20Name Asset%20Class) (Utilities|Paths|Combine (Utilities|Array|MakeArray (Utilities|String|Append "/" Plugin) Base%20Dir Base%20Name (CallFunction|GenerateBaseDirforClass Base%20Name Asset%20Class))) (Utilities|Paths|Combine (Utilities|Array|MakeArray (Utilities|Paths|Combine (Utilities|Array|MakeArray (Utilities|String|Append "/" Plugin) Base%20Dir Base%20Name (CallFunction|GenerateBaseDirforClass Base%20Name Asset%20Class))) (CallFunction|GenerateAssetNameforClass Base%20Name Asset%20Class)))))

(fn GenerateBaseDirForAsset
  (params (Base%20Name :type String) (Asset :type Object%20Reference))
  (return (CallFunction|GenerateBaseDirforClass Base%20Name (Utilities|GetClass Asset))))

(fn GenerateBaseDirForClass
  (params (Base%20Name :type String) (Asset%20Class :type Object%20Class%20Reference))
  (return (CallFunction|GetValueFromClassNameMap Asset%20Class (Variables|/AnimInEdTools/AssetBrowser/Blueprint/BP_AssetBrowser_NameMappings.BP_AssetBrowser_NameMappings_C|GetClass%20Base%20Dir%20Map (LocalVariables|GetClass%20Name%20Mappings)))))

(fn GetValuefromClassNameMap
  (params (Asset%20Class :type Object%20Class%20Reference) (Class%20Name%20Map :type Map%20of%20Object%20Soft%20Class%20References%20to%20Strings))
  (bind _output (Utilities|ResolveSoftReference Asset%20Class))
  (bind (_value _returnvalue) (Utilities|Map|Find Class%20Name%20Map _output))
  (if _returnvalue
    (return _value)
    (else
      (bind _keys (Utilities|Map|Keys (Variables|GetClass%20Name%20Map)))
      (for _array_element _keys
        (bind _returnvalue_2 (Utilities|LoadClassAssetBlocking _array_element))
        (if (Utilities|ClassIsChildOf (Variables|GetAsset%20Class) _returnvalue_2)
          (return (Utilities|Map|Find (Variables|GetClass%20Name%20Map) _array_element))))
      (return ""))))
