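; BPI_AssetBrowserCreate is a Blueprint interface in the AnimInEdTools asset browser, apparently for objects involved in creation.
; It declares CreateBreadcrumb, GetTitle for a title, and ValidateData, which returns a boolean with a string.
; The file holds only signatures with default returns, so implementing classes supply all behavior.
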(fn CreateBreadcrumb
  (return 0))

(fn GetTitle
  (return "None"))

(fn ValidateData
  (return false ""))
