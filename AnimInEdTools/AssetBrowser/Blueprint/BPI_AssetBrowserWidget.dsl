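; Blueprint Interface for asset browser widgets in the AnimInEdTools editor tooling.
; It declares a single function, SetAssetBreadcrumb, which passes a breadcrumb object of type ABT Abstract to the implementing widget.
; It holds no logic; implementers decide how the breadcrumb is used.
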(fn SetAssetBreadcrumb
  (params (Bread%20Crumb :type ABT%20Abstract%20Object%20Reference)))
