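# Editor utility widget that creates Asset Browser data assets for the AnimInEd tools, named from a text box.
# Buttons call AssetTools CreateAsset for abstract types, modules and name mappings (labelled Processor Settings Data), also setting Name on abstract types.
# Creature and prop buttons show no asset class and contain bad cast nodes, so look broken.
# A self-bound details view separately previews a generated asset path from plugin, directory, name and class inputs.
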EditorUtilityWidgetBlueprint /AnimInEdTools/AssetBrowser/Blueprint/EUW_AssetBrowserTypesMaker
graphs [EventGraph {nodes [
   "On Clicked (BuildAbstract_Button)"() {is=title K2Node_ComponentBoundEvent}
   > n1 = GetAssetTools()
   n2 = "To String (Text)"(InText=Get Text(self=Get Name_TextBox() -> (Name_TextBox:Editor Utility Editable Text Box Object Reference)))
   > n3 = CreateAsset(
    self=@n1
    AssetName=@n2
    PackagePath=/AnimInEdTools/AssetBrowser/Data/
    AssetClass=&/ABT_Abstract.ABT_Abstract_C
   )
   > n4 = Cast To ABT_Abstract!(execute, Object=@n3) -> (then, CastFailed, AsABT Abstract:ABT Abstract Object Reference) {is=title K2Node_DynamicCast}
   > Set Name!(execute, Name:String=@n2, self:ABT Abstract Object Reference=@n4) -> (then, Output_Get:String)
   "On Clicked (BuildCreatureType_Button)"() {is=title K2Node_ComponentBoundEvent}
   > n5 = GetAssetTools()
   n6 = "To String (Text)"(InText=Get Text(self=Get Name_TextBox()))
   > n7 = CreateAsset(self=@n5, AssetName=@n6, PackagePath=&Data/)
   > n8 = Bad cast node!(execute, Object=@n7) -> (then, CastFailed, AsABT Creature:Object Reference) {is=title K2Node_DynamicCast}
   > Set Name(Name=@n6, self=@n8)
   "On Clicked (BuildPropType_Button)"() {is=title K2Node_ComponentBoundEvent}
   > n9 = GetAssetTools()
   n10 = "To String (Text)"(InText=Get Text(self=Get Name_TextBox()))
   > n11 = CreateAsset(self=@n9, AssetName=@n10, PackagePath=&Data/)
   > n12 = Bad cast node!(execute, Object=@n11) -> (then, CastFailed, AsABT Emote Prop:Object Reference)
   > Set Name(Name=@n10, self=@n12)
   "On Clicked (BuildModuleType_Button)"() {is=title K2Node_ComponentBoundEvent}
   > n13 = GetAssetTools()
   n14 = Get Name_TextBox()
   > CreateAsset(
    self=@n13
    AssetName="To String (Text)"(InText=Get Text(self=@n14))
    PackagePath=&Data/
    AssetClass=/AnimInEdTools/AssetBrowser/Module/Base/BP_AssetBrowserModule.BP_AssetBrowserModule_C
   )
   "On Clicked (BuildProcessorSettingsDataType_Button)"() {is=title K2Node_ComponentBoundEvent}
   > n15 = GetAssetTools()
   n16 = Get Name_TextBox()
   > CreateAsset(
    self=@n15
    AssetName="To String (Text)"(InText=Get Text(self=@n16))
    PackagePath=&Data/
    AssetClass=&/BP_AssetBrowser_NameMappings.BP_AssetBrowser_NameMappings_C
   )
   "On Property Changed (DetailsView_0)"!() -> (OutputDelegate, then, PropertyName:Name=None) {is=title K2Node_ComponentBoundEvent}
   n17 = Get Class Type() -> (Class Type:Object Class Reference)
   n18 = Get Plugin Name() -> (Plugin Name:String)
   n19 = Get Base Dir() -> (Base Dir:String)
   n20 = Get Base Name() -> (Base Name:String)
   > Set Text(
    Text="To Text (String)"(InString=Generate Asset Path For Class!(Plugin:String=@n18, Base Dir:String=@n19, Base Name:String=@n20, Asset Class:Object Class Reference=@n17, __WorldContext:Object Reference) -> (Asset Name:String, Package Path:String, Full Path:String).Full Path)
    self=Get TextBlock_0() -> (TextBlock_0:Text Object Reference)
   )
   Event PreConstruct()
   > SetObject(self=Get DetailsView_0() -> (DetailsView_0:Details View Object Reference), NewObject=Self-Reference())
   Event Construct()
   Event Tick()
  ]}]
variables [
 Base Dir {type=string, default=Props/Emote}
 Base Name {type=string, default=AssetName}
 Class Type {type=Class, default=null}
 Plugin Name {type=string, default=JunoAnimInEdAssets}
]
widgets {
 info {rootWidgetClass=&VerticalBox, widgetCount=14}
 widgets [
  VerticalBox_0 {widgetClassPath=&VerticalBox, widget=VerticalBox_0}
  Name_TextBox {
   widgetClassPath=&EditorUtilityEditableTextBox
   bIsVariable=true
   parent=VerticalBox_0
   widget=Name_TextBox
   properties {text=Asset Name}
   slot {class=VerticalBoxSlot}
  }
  BuildAbstract_Button {
   widgetClassPath=&EditorUtilityButton
   parent=VerticalBox_0
   widget=BuildAbstract_Button
   properties {toolTipText=Build Abstract}
   slot {class=VerticalBoxSlot}
  }
  BuildAbstractTextBlock {
   widgetClassPath=&TextBlock
   parent=BuildAbstract_Button
   widget=BuildAbstractTextBlock
   properties {text=Build Abstract Type}
   slot {class=ButtonSlot}
  }
  BuildCreatureType_Button {
   widgetClassPath=&EditorUtilityButton
   parent=VerticalBox_0
   widget=BuildCreatureType_Button
   properties {toolTipText=Build Creature}
   slot {class=VerticalBoxSlot}
  }
  BuildCreatureTypeTextBlock {
   widgetClassPath=&TextBlock
   parent=BuildCreatureType_Button
   widget=BuildCreatureTypeTextBlock
   properties {text=Build Creature Type}
   slot {class=ButtonSlot}
  }
  BuildPropType_Button {
   widgetClassPath=&EditorUtilityButton
   parent=VerticalBox_0
   widget=BuildPropType_Button
   properties {toolTipText=Build Creature}
   slot {class=VerticalBoxSlot}
  }
  BuildPropTextBlock {
   widgetClassPath=&TextBlock
   parent=BuildPropType_Button
   widget=BuildPropTextBlock
   properties {text=Build Prop Type}
   slot {class=ButtonSlot}
  }
  BuildModuleType_Button {
   widgetClassPath=&EditorUtilityButton
   parent=VerticalBox_0
   widget=BuildModuleType_Button
   properties {toolTipText=Build Creature}
   slot {class=VerticalBoxSlot}
  }
  BuildModuleTextBlock {
   widgetClassPath=&TextBlock
   parent=BuildModuleType_Button
   widget=BuildModuleTextBlock
   properties {text=Build Module Type}
   slot {class=ButtonSlot}
  }
  BuildProcessorSettingsDataType_Button {
   widgetClassPath=&EditorUtilityButton
   parent=VerticalBox_0
   widget=BuildProcessorSettingsDataType_Button
   properties {toolTipText=Build Creature}
   slot {class=VerticalBoxSlot}
  }
  BuildProcessorSettingsDataTypeTextBlock {
   widgetClassPath=&TextBlock
   parent=BuildProcessorSettingsDataType_Button
   widget=BuildProcessorSettingsDataTypeTextBlock
   properties {text=Build Processor Settings Data Type}
   slot {class=ButtonSlot}
  }
  DetailsView_0 {
   widgetClassPath=&DetailsView
   parent=VerticalBox_0
   widget=DetailsView_0
   properties {categoriesToShow [Inputs]}
   slot {class=VerticalBoxSlot}
  }
  TextBlock_0 {
   widgetClassPath=&TextBlock
   bIsVariable=true
   parent=VerticalBox_0
   widget=TextBlock_0
   properties {
    text=Path
    font {
     bForceMonospaced=false
     bMaterialIsStencil=false
     fontMaterial=null
     fontObject=&Roboto
     letterSpacing=0
     monospacedWidth=1
     size=15
     skewAmount=0
     typefaceFontName=Bold
     outlineSettings {
      bApplyOutlineToDropShadows=false
      bMiteredCorners=false
      bSeparateFillAlpha=false
      outlineMaterial=null
      outlineSize=0
      outlineColor=<rgba 0 0 0 1>
     }
    }
   }
   slot {class=VerticalBoxSlot}
  }
 ]
}
