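# Editor utility widget for the AnimInEdTools asset browser that provides a file path input field.
# A DetailsView in a vertical box targets the widget itself on PreConstruct and is limited to the Inputs category, using the property editor as the UI.
# Any property change in that view broadcasts the On File Path Changed dispatcher with the current File Path string.
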EditorUtilityWidgetBlueprint /AnimInEdTools/AssetBrowser/Blueprint/EUW_AssetBrowser_FilePath
graphs [
 EventGraph {nodes [
   "On Property Changed (DetailsView)"!() -> (OutputDelegate, then, PropertyName:Name=None) {is=title K2Node_ComponentBoundEvent}
   > Call On File Path Changed!(execute:Exec, self:EUW Asset Browser File Path Object Reference, File Path:String=Get!() -> (File Path_FilePath:String)) -> (then:Exec) {is=title K2Node_CallDelegate}
   Event PreConstruct()
   > SetObject(self=Get DetailsView() -> (DetailsView:Details View Object Reference), NewObject=Self-Reference())
   Event Construct()
  ]}
 On File Path Changed {nodes [On File Path Changed!() -> (then, File Path:String) {is=title K2Node_FunctionEntry}]}
]
variables [
 File Path {type=FilePath, default {filePath=""}}
 On File Path Changed {default=()}
]
widgets {
 info {rootWidgetClass=&VerticalBox, widgetCount=2}
 widgets [
  VerticalBox_0 {widgetClassPath=&VerticalBox, widget=VerticalBox_0}
  DetailsView {
   widgetClassPath=&DetailsView
   parent=VerticalBox_0
   widget=DetailsView
   properties {categoriesToShow [Inputs]}
   slot {class=VerticalBoxSlot}
  }
 ]
}
